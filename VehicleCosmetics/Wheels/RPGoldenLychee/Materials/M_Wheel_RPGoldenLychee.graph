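# Translucent Default Lit material for an animated flipbook effect on the RPGoldenLychee cosmetic wheel, flagged for skeletal mesh and mesh deformer use.
# Time multiplied by a Speed parameter drives the FlipBook function's phase, with row and column parameters defining a single-row strip by default.
# The resulting UVs sample the rim FX texture parameter, which uses explicit mip level mode.
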Material /VehicleCosmetics/Wheels/RPGoldenLychee/Materials/M_Wheel_RPGoldenLychee
graph {
 properties {
  bAutomaticallySetUsageInEditor=false
  bUsedWithMeshDeformer=true
  bUsedWithSkeletalMesh=true
  blendMode=BLEND_Translucent
  usedShadingModels=Default Lit
 }
 nodes [
  n1 = Time()
  n2 = ScalarParameter() {defaultValue=0.5, parameterName=Speed}
  TextureSampleParameter2D!(
   UVs=MaterialFunctionCall(
    "Animation  Phase (0-1)"=Multiply(A=@n1, B=@n2)
    Number of Rows=ScalarParameter() {defaultValue=1, parameterName=NumRows}
    Number of Columns=ScalarParameter() {defaultValue=9, parameterName=NumColumns}
    UVs=TextureCoordinate()
   ) {materialFunction=&FlipBook}.UVs
   Level
   Apply View MipBias
  ) {
   mipValueMode=TMVM_MipLevel
   parameterName=Texture_FlipBook
   texture=/VehicleCosmetics/Wheels/RPGoldenLychee/Textures/T_Rim_RPGoldenLychee_FX_BC
  }
 ]
}
